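; Parent Blueprint for an industrial chain hoist prop: a mount mesh with a movable trolley, a chain of adjustable length and a hook.
; The construction script measures mesh bounds, positions the trolley along X using a clamped or random position, builds the chain, then places the hook below it.
; Chain links are instanced static meshes stacked from the trolley socket with incremental rotation.
; Helpers for capsule collision, tint, materials and light placement are defined but not called by construction.
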(event EventBeginPlay)

(event Collision|EventActorBeginOverlap
  (params (OtherActor :type Actor%20Object%20Reference)))

(event EventTick
  (params (DeltaSeconds :type Float%20%28single-precision%29)))

(fn AddCapsuleCollision
  (bind _returnvalue (AddComponent|NODE_AddCapsuleComponent-3 :RelativeTransform (Math|Transform|MakeTransform (Math|Vector|MakeVector 0.0 0.0 (- (/ (Variables|ChainHoist|GetChainLength) 2.0) (- (Variables|ChainHoist|GetChainLength) (Variables|AdjustableLight|GetMountBounds)))))))
  (bind _output_get (Variables|Default|SetCapsuleCollision _returnvalue))
  (Components|Capsule|SetCapsuleHalfHeight _output_get (/ (Variables|ChainHoist|GetChainLength) 2.0) true)
  (for _array_element (Utilities|Array|MakeArray (Variables|AdjustableLight|GetHoistHook) (Variables|AdjustableLight|GetHoistChain) (Variables|Debug|GetStaticMeshComponent))
    (Collision|SetCollisionResponsetoAllChannels _array_element "ECR_Ignore")
    (Collision|SetCollisionEnabled _array_element "NoCollision")))

(fn EnableCollisionOnMeshParts
  (for _array_element (Utilities|Array|MakeArray (Variables|AdjustableLight|GetHoistChain) (Variables|AdjustableLight|GetHoistHook))
    (Collision|SetCollisionProfileName _array_element "Default" true :!class /Script/Engine.PrimitiveComponent)
    (Development|PrintString (Utilities|String|ToString(Name) (Collision|GetCollisionProfileName (stale _array_element_2)))))
  (for _array_element_2 (Utilities|Array|MakeArray (Variables|AdjustableLight|GetHoistChain) (Variables|AdjustableLight|GetHoistHook))
    _))

(fn GetCurieCollisionComponentOverride
  (return (Variables|Default|GetCapsuleCollision)))

(fn GetMaterialFromRootMesh
  (Rendering|Material|SetMaterial (Variables|AdjustableLight|GetHoistHook) 0 (Material|GetMID (Variables|Debug|GetStaticMeshComponent))))

(fn GetMeshBounds
  (Variables|AdjustableLight|SetMountBounds (.z (Components|StaticMesh|GetLocalBounds (Variables|Debug|GetStaticMeshComponent))))
  (Variables|AdjustableLight|SetHookBounds (.z (Components|StaticMesh|GetLocalBounds (Variables|AdjustableLight|GetHoistHook))))
  (Variables|AdjustableLight|SetChainBounds (.z (Components|StaticMesh|GetLocalBounds (Variables|AdjustableLight|GetHoistChain)))))

(fn GetTintColorFromTextureData
  (bind _outtexturedataarray (Building|GetBuildingTextureData))
  (Variables|AdjustableLight|SetTintColor (Math|Conversions|ToLinearColor(Color) (Variables|/Script/FortniteGame.BuildingTextureData|GetTintColor (Utilities|Array|Get(acopy) _outtexturedataarray)))))

(fn SetChainPosition
  (bind _chainbounds (Variables|AdjustableLight|GetChainBounds))
  (bind _returnvalue (* _chainbounds -1.0))
  (if (!= _returnvalue 0)
    (bind _output_get (Variables|AdjustableLight|SetRodScale (/ (Utilities|Operators|Subtract (Variables|ChainHoist|GetChainLength) (* (Variables|AdjustableLight|GetMountBounds) -1.0) (* (Variables|AdjustableLight|GetHookBounds) -1.0)) _returnvalue)))
    (Transformation|SetRelativeScale3D (Variables|AdjustableLight|GetHoistChain) (Math|Vector|MakeVector 1.0 1.0 _output_get))
    (Transformation|SetRelativeLocation (Variables|AdjustableLight|GetHoistChain) (Math|Vector|MakeVector 0.0 0.0 (Variables|AdjustableLight|GetMountBounds)) false false)))

(fn SetDownRodMaterial
  (bind _hoistchain (Variables|AdjustableLight|GetHoistChain))
  (bind _returnvalue (Rendering|Material|GetMaterial _hoistchain 0))
  (bind _hoistchain_2 (Variables|AdjustableLight|GetHoistChain))
  (bind _returnvalue_2 (Rendering|Material|GetMaterial _hoistchain_2 0))
  (if (Variables|AdjustableLight|GetUseGeoForDownrod)
    (if (Utilities|IsValid _returnvalue_2 :!class /Script/Engine.KismetSystemLibrary)
      (Rendering|Material|CreateDynamicMaterialInstance (Variables|AdjustableLight|GetISM_ChainTrolley) 0 _returnvalue_2 "None" :!class /Script/Engine.PrimitiveComponent))
    (elif (Utilities|IsValid _returnvalue :!class /Script/Engine.KismetSystemLibrary)
      (bind _returnvalue_3 (Rendering|Material|CreateDynamicMaterialInstance (Variables|AdjustableLight|GetHoistChain) 0 _returnvalue "None" :!class /Script/Engine.PrimitiveComponent))
      (Rendering|Material|SetScalarParameterValue _returnvalue_3 "UV Scale Vertical" (Variables|AdjustableLight|GetRodScale) :!class /Script/Engine.MaterialInstanceDynamic))))

(fn SetFixturePosition
  (bind _hoisttrolley (Variables|Default|GetHoistTrolley))
  (bind _returnvalue (Transformation|GetAllSocketNames _hoisttrolley))
  (if (> (Utilities|Array|Length _returnvalue) 0)
    (Transformation|SetRelativeLocation (Variables|AdjustableLight|GetHoistHook) (Math|Vector|MakeVector (.x (.location (Transformation|GetSocketTransform (Variables|Default|GetHoistTrolley) (Utilities|Array|Get(acopy) _returnvalue) "RTS_Actor"))) (.y (.location (Transformation|GetSocketTransform (Variables|Default|GetHoistTrolley) (Utilities|Array|Get(acopy) _returnvalue) "RTS_Actor"))) (+ (.z (.location (Transformation|GetSocketTransform (Variables|Default|GetHoistTrolley) (Utilities|Array|Get(acopy) _returnvalue) "RTS_Actor"))) (* (Variables|Default|GetNumberofChains) (Variables|AdjustableLight|GetChainBounds)))) false false)))

(fn SetLightPosition
  (bind _returnvalue (Actor|GetComponentsByClass "/Script/Engine.LightComponent"))
  (if (> (Utilities|Array|Length _returnvalue) 0)
    (Transformation|SetWorldTransform (Utilities|Array|Get(acopy) _returnvalue) (Transformation|GetSocketTransform (Variables|AdjustableLight|GetHoistHook) (Utilities|Array|Get(acopy) (Transformation|GetAllSocketNames (Variables|AdjustableLight|GetHoistHook))) "RTS_World") false false)))

(fn SetNumberofChainMeshes
  (bind _hoistchain (Variables|AdjustableLight|GetHoistChain))
  (bind _staticmesh (Variables|/Script/Engine.StaticMeshComponent|GetStaticMesh _hoistchain))
  (bind _returnvalue (AddComponent|NODE_AddInstancedStaticMeshComponent-0))
  (bind _output_get (Variables|AdjustableLight|SetISM_ChainTrolley _returnvalue))
  (if (Utilities|IsValid _staticmesh :!class /Script/Engine.KismetSystemLibrary)
    (Components|StaticMesh|SetStaticMesh _output_get _staticmesh)
    (bind _output_get_2 (LocalVariables|SetNumberOfInstances (Math|Conversions|ToFloat(Integer) (Math|Float|Round (Math|Conversions|ToFloat(Integer) (Math|Float|Round (/ (- (Variables|ChainHoist|GetChainLength) (Variables|AdjustableLight|GetHookBounds)) (* (Variables|AdjustableLight|GetChainBounds) -1.0))))))))
    (for _index (range (+ (Math|Float|Truncate _output_get_2) 1))
      (if (== _index 0)
        (Components|InstancedStaticMesh|AddInstance (Variables|AdjustableLight|GetISM_ChainTrolley) (Transformation|GetSocketTransform (Variables|Default|GetHoistTrolley) (Utilities|Array|Get(acopy) (Transformation|GetAllSocketNames (Variables|Default|GetHoistTrolley))) "RTS_Actor") false)
        (else
          (Components|InstancedStaticMesh|AddInstance (Variables|AdjustableLight|GetISM_ChainTrolley) (Math|Transform|MakeTransform (Math|Vector|MakeVector (.x (.location (Transformation|GetSocketTransform (Variables|Default|GetHoistTrolley) (Utilities|Array|Get(acopy) (Transformation|GetAllSocketNames (Variables|Default|GetHoistTrolley))) "RTS_Actor"))) (.y (.location (Transformation|GetSocketTransform (Variables|Default|GetHoistTrolley) (Utilities|Array|Get(acopy) (Transformation|GetAllSocketNames (Variables|Default|GetHoistTrolley))) "RTS_Actor"))) (+ (+ (* _index (Variables|AdjustableLight|GetChainBounds)) (Variables|AdjustableLight|GetChainBounds)) (* (Variables|AdjustableLight|GetRodMeshOffset) -1.0))) (Math|Rotator|MakeRotator 0.0 0.0 (Math|Conversions|ToFloat(Integer) (* _index (Variables|Default|GetChainMeshRotationAmount))))) false)
          (Rendering|SetVisibility (Variables|AdjustableLight|GetHoistChain) false false)))
      (Variables|Default|SetNumberofChains (LocalVariables|GetNumberOfInstances)))
    (else
      (Development|PrintString "DownRod mesh not found"))))

(fn SetTintColorOnMeshParts
  (for _array_element (Utilities|Array|MakeArray (Variables|AdjustableLight|GetHoistChain) (Variables|AdjustableLight|GetHoistHook))
    (Default|SetColoronSecondaryMesh (Variables|AdjustableLight|GetTintColor) _array_element)))

(fn SetTrolleyPosition
  (bind (_min _max) (Components|StaticMesh|GetLocalBounds (Variables|Default|GetHoistTrolley)))
  (bind (_min_2 _max_2) (Components|StaticMesh|GetLocalBounds (Variables|Debug|GetStaticMeshComponent)))
  (bind _chainposition (Variables|ChainHoist|GetChainPosition))
  (LocalVariables|SetL_AdjustedTrolleyOffset (Math|Conversions|ToFloat(Integer) (+ (.x _max) 10)))
  (LocalVariables|SetL_MountBounds (.x _max_2))
  (if (Variables|ChainHoist|GetUseRandomPosition)
    (Transformation|SetRelativeTransform (Variables|Default|GetHoistTrolley) (Math|Transform|MakeTransform (Math|Vector|MakeVector (Math|Random|RandomFloatinRange (LocalVariables|GetL_AdjustedTrolleyOffset) (- (LocalVariables|GetL_MountBounds) (- (LocalVariables|GetL_AdjustedTrolleyOffset) 1e+01))))) false false)
    (elif (Math|Float|InRange(Float) (Variables|ChainHoist|GetChainPosition) 9e+01 893.0)
      _
      (elif (< _chainposition 9e+01)
        (Variables|ChainHoist|SetChainPosition 9e+01)
        (elif (> _chainposition 893.0)
          (Variables|ChainHoist|SetChainPosition 893.0)))))
  (Transformation|SetRelativeTransform (Variables|Default|GetHoistTrolley) (Math|Transform|MakeTransform (Math|Vector|MakeVector (Variables|ChainHoist|GetChainPosition))) false false))

(fn ConstructionScript
  (CallFunction|GetMeshBounds)
  (CallFunction|SetTrolleyPosition)
  (CallFunction|SetNumberofChainMeshes)
  (CallFunction|SetFixturePosition))

(fn ValidateBuildingTextureData
  (bind _outtexturedataarray (Building|GetBuildingTextureData))
  (Variables|AdjustableLight|SetHasTextureData (> (Utilities|Array|Length _outtexturedataarray) 0))
  (return))
